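# Physics asset for the large male Bengal Basher suit head, providing secondary motion bodies for its whiskers and ears.
# A head capsule anchors six two-segment whisker chains (low, mid and up on each side) built from thin capsules, plus two ear bodies with zero-length capsules.
# Constraints chain each outer whisker segment to its inner one and on to the head, allowing moderate swing with almost no twist; ears swing less and have no twist.
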PhysicsAsset /BRCosmetics/Characters/Player/Male/Large/Heads/M_LRG_BengalBasher_Suit_Head/Meshes/M_LRG_BengalBasher_Suit_Head_Physics
bodies {
 dyn_1_L_low_whisker_joint [{shapeType=Capsule, length=6, radius=0.5, shapeName=dyn_1_L_low_whisker_joint_capsule, center=<3.1816 0 0>, rotation=<rot -90 0 -0.0002>}]
 dyn_1_L_mid_whisker_joint [{
   shapeType=Capsule
   length=7.5
   radius=0.5
   shapeName=dyn_1_L_mid_whisker_joint_capsule
   center=<4.0301 0 0>
   rotation=<rot -90 179.789 -179.7892>
  }]
 dyn_1_L_up_whisker_joint [{
   shapeType=Capsule
   length=7
   radius=0.5
   shapeName=dyn_1_L_up_whisker_joint_capsule
   center=<4.0009 -0.0001 0>
   rotation=<rot -90 90 -90.0009>
  }]
 dyn_1_R_low_whisker_joint [{
   shapeType=Capsule
   length=4.5
   radius=0.5
   shapeName=dyn_1_R_low_whisker_joint_capsule
   center=<2.6926 0 0>
   rotation=<rot 90 -179.3669 -179.3662>
  }]
 dyn_1_R_mid_whisker_joint [{shapeType=Capsule, length=7, radius=0.5, shapeName=dyn_1_R_mid_whisker_joint_capsule, center=<3.7226 0 0>, rotation=<rot 90 10.0252 10.0259>}]
 dyn_1_R_up_whisker_joint [{
   shapeType=Capsule
   length=6
   radius=0.5
   shapeName=dyn_1_R_up_whisker_joint_capsule
   center=<3.2611 0 0>
   rotation=<rot 90 179.8416 179.8419>
  }]
 dyn_2_L_low_whisker_joint [{
   shapeType=Capsule
   length=6.5
   radius=0.5
   shapeName=dyn_2_L_low_whisker_joint_capsule
   center=<4.0324 0 0>
   rotation=<rot -90 179.9999 179.9998>
  }]
 dyn_2_L_mid_whisker_joint [{
   shapeType=Capsule
   length=6
   radius=0.5
   shapeName=dyn_2_L_mid_whisker_joint_capsule
   center=<3.661 0 0>
   rotation=<rot -90 166.7373 -166.7376>
  }]
 dyn_2_L_up_whisker_joint [{shapeType=Capsule, length=7, radius=0.5, shapeName=dyn_2_L_up_whisker_joint_capsule, center=<3.4298 0 0>, rotation=<rot -90 0 -0.0002>}]
 dyn_2_R_low_whisker_joint [{
   shapeType=Capsule
   length=6.5
   radius=0.5
   shapeName=dyn_2_R_low_whisker_joint_capsule
   center=<3.7907 0 0>
   rotation=<rot 90 179.3668 179.3673>
  }]
 dyn_2_R_mid_whisker_joint [{
   shapeType=Capsule
   length=7
   radius=0.5
   shapeName=dyn_2_R_mid_whisker_joint_capsule
   center=<4.0762 0 0>
   rotation=<rot 90 179.9472 179.9475>
  }]
 dyn_2_R_up_whisker_joint [{
   shapeType=Capsule
   length=7.5
   radius=0.5
   shapeName=dyn_2_R_up_whisker_joint_capsule
   center=<4.2427 -0.0001 0>
   rotation=<rot 90 90 90.0009>
  }]
 dyn_L_ear_joint [{shapeType=Capsule, length=0, radius=3.5, shapeName=dyn_L_ear_joint_capsule, center=<7.0197 -0.0001 0>, rotation=<rot -90 179.9978 -179.9989>}]
 dyn_R_ear_joint [{shapeType=Capsule, length=0, radius=3.5, shapeName=dyn_R_ear_joint_capsule, center=<7.1672 -0.0001 0>, rotation=<rot 90 166.7374 166.7379>}]
 head [{shapeType=Capsule, length=12, radius=5.5, shapeName=head_capsule}]
}
constraints [
 {bone1Name=dyn_1_R_low_whisker_joint, bone2Name=head, swing1LimitDegrees=30, swing2LimitDegrees=30, twistLimitDegrees=1}
 {bone1Name=dyn_1_R_mid_whisker_joint, bone2Name=head, swing1LimitDegrees=30, swing2LimitDegrees=30, twistLimitDegrees=1}
 {bone1Name=dyn_1_R_up_whisker_joint, bone2Name=head, swing1LimitDegrees=30, swing2LimitDegrees=30, twistLimitDegrees=1}
 {bone1Name=dyn_1_L_low_whisker_joint, bone2Name=head, swing1LimitDegrees=30, swing2LimitDegrees=30, twistLimitDegrees=1}
 {bone1Name=dyn_1_L_mid_whisker_joint, bone2Name=head, swing1LimitDegrees=30, swing2LimitDegrees=30, twistLimitDegrees=1}
 {bone1Name=dyn_1_L_up_whisker_joint, bone2Name=head, swing1LimitDegrees=30, swing2LimitDegrees=30, twistLimitDegrees=1}
 {bone1Name=dyn_2_L_low_whisker_joint, bone2Name=dyn_1_L_low_whisker_joint, swing1LimitDegrees=30, swing2LimitDegrees=30, twistLimitDegrees=1}
 {bone1Name=dyn_2_L_mid_whisker_joint, bone2Name=dyn_1_L_mid_whisker_joint, swing1LimitDegrees=30, swing2LimitDegrees=30, twistLimitDegrees=1}
 {bone1Name=dyn_2_L_up_whisker_joint, bone2Name=dyn_1_L_up_whisker_joint, swing1LimitDegrees=30, swing2LimitDegrees=30, twistLimitDegrees=1}
 {bone1Name=dyn_2_R_low_whisker_joint, bone2Name=dyn_1_R_low_whisker_joint, swing1LimitDegrees=30, swing2LimitDegrees=30, twistLimitDegrees=1}
 {bone1Name=dyn_2_R_mid_whisker_joint, bone2Name=dyn_1_R_mid_whisker_joint, swing1LimitDegrees=30, swing2LimitDegrees=30, twistLimitDegrees=1}
 {bone1Name=dyn_2_R_up_whisker_joint, bone2Name=dyn_1_R_up_whisker_joint, swing1LimitDegrees=30, swing2LimitDegrees=30, twistLimitDegrees=1}
 {bone1Name=dyn_L_ear_joint, bone2Name=head, swing1LimitDegrees=20, swing2LimitDegrees=15, twistLimitDegrees=0}
 {bone1Name=dyn_R_ear_joint, bone2Name=head, swing1LimitDegrees=20, swing2LimitDegrees=15, twistLimitDegrees=0}
]
