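# Post-style head AnimBP for the male large Bengal Basher suit head, adding procedural whisker motion on top of the attached parent mesh's copied pose.
# Six chained parent constraints drive the left and right upper, mid and lower whisker base joints, each blending eight head and face bones by fixed weights so whiskers follow expressions.
# A RigidBody node then simulates the head's physics asset relative to the head bone, with damped component motion.
# The event graph holds no logic.
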AnimBlueprint /BRCosmetics/Characters/Player/Male/Large/Heads/M_LRG_BengalBasher_Suit_Head/Meshes/M_LRG_BengalBasher_Suit_Head_AnimBP_New
target_skeleton=&/M_LRG_BengalBasher_Suit_Head_Skeleton
graphs [
 AnimGraph {nodes [
   n1 = "Constraint - L_up_whisker_base_joint (PPPPPPPP)"!(
    ComponentPose=Local To Component(LocalPose=Copy Pose From Mesh() {bUseAttachedParent=true})
    Alpha=1
    ConstraintWeights_0=0.6882
    ConstraintWeights_1:float=0.1296
    ConstraintWeights_2:float=0.1171
    ConstraintWeights_3:float=0.0274
    ConstraintWeights_4:float=0.0143
    ConstraintWeights_5:float=0.0126
    ConstraintWeights_6:float=0.009
    ConstraintWeights_7:float=0.0017
   ) -> (Pose) {
    alphaBoolBlend {blendOption=Linear}
    boneToModify {boneName=L_up_whisker_base_joint}
    constraintSetup [
     {targetBone {boneName=head}}
     {targetBone {boneName=C_nose_bridge}}
     {targetBone {boneName=L_cheek_inner}}
     {targetBone {boneName=C_lip_upper_mid}}
     {targetBone {boneName=C_jaw}}
     {targetBone {boneName=L_lip_upper_outer}}
     {targetBone {boneName=L_lip_corner}}
     {targetBone {boneName=L_eye_lid_lower_mid}}
    ]
    constraintWeights [1,1,1,1,1,1,1,1]
   }
   n2 = "Constraint - L_mid_whisker_base_joint (PPPPPPPP)"!(
    ComponentPose=@n1
    Alpha=1
    ConstraintWeights_0=0.6218
    ConstraintWeights_1:float=0.1932
    ConstraintWeights_2:float=0.1121
    ConstraintWeights_3:float=0.0489
    ConstraintWeights_4:float=0.0154
    ConstraintWeights_5:float=0.0068
    ConstraintWeights_6:float=0.0016
    ConstraintWeights_7:float=0.0003
   ) -> (Pose) {
    alphaBoolBlend {blendOption=Linear}
    boneToModify {boneName=L_mid_whisker_base_joint}
    constraintSetup [
     {targetBone {boneName=head}}
     {targetBone {boneName=C_nose_bridge}}
     {targetBone {boneName=C_lip_upper_mid}}
     {targetBone {boneName=L_lip_upper_outer}}
     {targetBone {boneName=L_cheek_inner}}
     {targetBone {boneName=C_jaw}}
     {targetBone {boneName=R_cheek_inner}}
     {targetBone {boneName=L_lip_corner}}
    ]
    constraintWeights [1,1,1,1,1,1,1,1]
   }
   n3 = "Constraint - L_low_whisker_base_joint (PPPPPPPP)"!(
    ComponentPose=@n2
    Alpha=1
    ConstraintWeights_0=0.5005
    ConstraintWeights_1:float=0.178
    ConstraintWeights_2:float=0.1247
    ConstraintWeights_3:float=0.0636
    ConstraintWeights_4:float=0.0525
    ConstraintWeights_5:float=0.0408
    ConstraintWeights_6:float=0.0394
    ConstraintWeights_7:float=0.0006
   ) -> (Pose) {
    alphaBoolBlend {blendOption=Linear}
    boneToModify {boneName=L_low_whisker_base_joint}
    constraintSetup [
     {targetBone {boneName=head}}
     {targetBone {boneName=L_lip_upper_outer}}
     {targetBone {boneName=C_lip_upper_mid}}
     {targetBone {boneName=L_lip_corner}}
     {targetBone {boneName=C_nose_bridge}}
     {targetBone {boneName=L_cheek_inner}}
     {targetBone {boneName=C_jaw}}
     {targetBone {boneName=R_cheek_inner}}
    ]
    constraintWeights [1,1,1,1,1,1,1,1]
   }
   n4 = "Constraint - R_up_whisker_base_joint (PPPPPPPP)"!(
    ComponentPose=@n3
    Alpha=1
    ConstraintWeights_0=0.6826
    ConstraintWeights_1:float=0.1802
    ConstraintWeights_2:float=0.0548
    ConstraintWeights_3:float=0.0433
    ConstraintWeights_4:float=0.0256
    ConstraintWeights_5:float=0.0099
    ConstraintWeights_6:float=0.0029
    ConstraintWeights_7:float=0.0006
   ) -> (Pose) {
    alphaBoolBlend {blendOption=Linear}
    boneToModify {boneName=R_up_whisker_base_joint}
    constraintSetup [
     {targetBone {boneName=head}}
     {targetBone {boneName=C_nose_bridge}}
     {targetBone {boneName=R_cheek_inner}}
     {targetBone {boneName=C_lip_upper_mid}}
     {targetBone {boneName=R_lip_upper_outer}}
     {targetBone {boneName=C_jaw}}
     {targetBone {boneName=R_lip_corner}}
     {targetBone {boneName=L_cheek_inner}}
    ]
    constraintWeights [1,1,1,1,1,1,1,1]
   }
   n5 = "Constraint - R_mid_whisker_base_joint (PPPPPPPP)"!(
    ComponentPose=@n4
    Alpha=1
    ConstraintWeights_0=0.6092
    ConstraintWeights_1:float=0.1877
    ConstraintWeights_2:float=0.1276
    ConstraintWeights_3:float=0.0519
    ConstraintWeights_4:float=0.0146
    ConstraintWeights_5:float=0.0071
    ConstraintWeights_6:float=0.0016
    ConstraintWeights_7:float=0.0003
   ) -> (Pose) {
    alphaBoolBlend {blendOption=Linear}
    boneToModify {boneName=R_mid_whisker_base_joint}
    constraintSetup [
     {targetBone {boneName=head}}
     {targetBone {boneName=C_nose_bridge}}
     {targetBone {boneName=C_lip_upper_mid}}
     {targetBone {boneName=R_lip_upper_outer}}
     {targetBone {boneName=R_cheek_inner}}
     {targetBone {boneName=C_jaw}}
     {targetBone {boneName=L_cheek_inner}}
     {targetBone {boneName=C_lip_lower_mid}}
    ]
    constraintWeights [1,1,1,1,1,1,1,1]
   }
   n6 = "Constraint - R_low_whisker_base_joint (PPPPPPPP)"!(
    ComponentPose=@n5
    Alpha=1
    ConstraintWeights_0=0.5028
    ConstraintWeights_1:float=0.1764
    ConstraintWeights_2:float=0.1248
    ConstraintWeights_3:float=0.062
    ConstraintWeights_4:float=0.0537
    ConstraintWeights_5:float=0.0407
    ConstraintWeights_6:float=0.0389
    ConstraintWeights_7:float=0.0007
   ) -> (Pose) {
    alphaBoolBlend {blendOption=Linear}
    boneToModify {boneName=R_low_whisker_base_joint}
    constraintSetup [
     {targetBone {boneName=head}}
     {targetBone {boneName=R_lip_upper_outer}}
     {targetBone {boneName=C_lip_upper_mid}}
     {targetBone {boneName=R_lip_corner}}
     {targetBone {boneName=C_nose_bridge}}
     {targetBone {boneName=R_cheek_inner}}
     {targetBone {boneName=C_jaw}}
     {targetBone {boneName=L_cheek_inner}}
    ]
    constraintWeights [1,1,1,1,1,1,1,1]
   }
   n7 = RigidBody(ComponentPose=@n6) {
    overridePhysicsAsset=&/M_LRG_BengalBasher_Suit_Head_Physics
    simulationSpace=BaseBoneSpace
    alphaBoolBlend {blendOption=Linear}
    baseBoneRef {boneName=head}
    componentAppliedLinearAccClamp=<3000 3000 3000>
    componentLinearAccScale=<0.4 0.4 0.4>
    componentLinearVelScale=<0.4 0.4 0.4>
   }
   Output Pose(Result=Component To Local(ComponentPose=@n7)) {layerGroup=None}
  ]}
 EventGraph {nodes [
   TryGetPawnOwner()
   Event BlueprintUpdateAnimation()
  ]}
]
variables []
